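# Sound cue for a job board completion in the shared quests UI, played in the non-critical in-game UI sound class.
# A mixer root layers two wave players at equal volume, one for the completion whoosh and one for the completed sound.
# Both waves play together as a single cue, with no randomization or modulation.
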SoundCue /Game/Sounds/Fort_UI_Sounds/SharedQuests/JobBoard_Whoosh_Cue
root=Mixer_0
nodes [
 {id=Mixer_0, children [WavePlayer_0,WavePlayer_3], properties {inputVolume [1,1]}}
 {id=WavePlayer_0, properties {soundWaveAssetPtr=&/JobBoard_Complete_Whoosh}}
 {id=WavePlayer_3, properties {soundWaveAssetPtr=&/JobBoard_Completed}}
]
properties {maxDistance=2097152, soundClassObject=&Fort_UI_InGame_NonCritical}
