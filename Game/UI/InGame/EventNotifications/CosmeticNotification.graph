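# Toast notification widget, derived from FortNotificationEntry_LevelUp, that shows a level-up or quest cosmetic reward with header, description and item icon.
# Intro and Outro animations play on notification events and report completion back, with a whoosh sound firing once.
# Both style functions set the same blue border color, ignoring quest rarity, and differ only in header casing.
# Construct sets container opacity higher for frontend controllers, and PreConstruct enlarges the widget on mobile.
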WidgetBlueprint /Game/UI/InGame/EventNotifications/CosmeticNotification
animations [
 Intro {
  binding_count=8
  length_seconds=0.75
  object=&self:Intro
  sequence {bindings [
    Text_Description {
     id=#1
     possessed_class=CommonTextBlock
     tracks [{
       type=MovieScene2DTransformTrack
       sections [{
         type=MovieScene2DTransformSection
         channels [
          Translation.X {type=MovieSceneScriptingFloatChannel, keys [[1,-32],[6,0]]}
          Translation.Y {type=MovieSceneScriptingFloatChannel, const [0,1]}
          Angle {type=MovieSceneScriptingFloatChannel}
          Scale.X {type=MovieSceneScriptingFloatChannel}
          Scale.Y {type=MovieSceneScriptingFloatChannel}
          Shear.X {type=MovieSceneScriptingFloatChannel}
          Shear.Y {type=MovieSceneScriptingFloatChannel}
         ]
        }]
      }]
    }
    CommonBorder_Content {
     id=#2
     possessed_class=CommonBorder
     tracks [
      {
       type=MovieSceneFloatTrack
       display_name=Render Opacity
       sections [{type=MovieSceneFloatSection, channels [None {type=MovieSceneScriptingFloatChannel, keys [[0,0],[5,1]]}]}]
      }
      {
       type=MovieScene2DTransformTrack
       sections [{
         type=MovieScene2DTransformSection
         channels [
          Translation.X {type=MovieSceneScriptingFloatChannel, keys [[0,-64],[5,0]]}
          Translation.Y {type=MovieSceneScriptingFloatChannel}
          Angle {type=MovieSceneScriptingFloatChannel}
          Scale.X {type=MovieSceneScriptingFloatChannel}
          Scale.Y {type=MovieSceneScriptingFloatChannel}
          Shear.X {type=MovieSceneScriptingFloatChannel}
          Shear.Y {type=MovieSceneScriptingFloatChannel}
         ]
        }]
      }
     ]
    }
    Reward_Preview_Shadow {
     id=#3
     possessed_class=FortLazyImage
     tracks [{
       type=MovieSceneFloatTrack
       display_name=Render Opacity
       sections [{type=MovieSceneFloatSection, channels [None {type=MovieSceneScriptingFloatChannel, const [0,1]}]}]
      }]
    }
    Overlay_0 {
     id=#4
     possessed_class=Overlay
     tracks [{
       type=MovieSceneFloatTrack
       display_name=Render Opacity
       sections [{type=MovieSceneFloatSection, end_frame=0, start_frame=0, channels [None {type=MovieSceneScriptingFloatChannel, const [1,1]}]}]
      }]
    }
    Reward_Glow {
     id=#5
     possessed_class=Image
     tracks [
      {
       type=MovieSceneFloatTrack
       display_name=Render Opacity
       sections [{type=MovieSceneFloatSection, channels [None {type=MovieSceneScriptingFloatChannel, keys [[4,0],[6,0.4295],[10,0.1]]}]}]
      }
      {
       type=MovieScene2DTransformTrack
       sections [{
         type=MovieScene2DTransformSection
         channels [
          Translation.X {type=MovieSceneScriptingFloatChannel}
          Translation.Y {type=MovieSceneScriptingFloatChannel}
          Angle {type=MovieSceneScriptingFloatChannel}
          Scale.X {type=MovieSceneScriptingFloatChannel, keys [[4,1.5],[6,2],[7,2]]}
          Scale.Y {type=MovieSceneScriptingFloatChannel, keys [[4,0.75],[6,1.5],[7,1.5]]}
          Shear.X {type=MovieSceneScriptingFloatChannel}
          Shear.Y {type=MovieSceneScriptingFloatChannel}
         ]
        }]
      }
     ]
    }
    RewardFlare {
     id=#6
     possessed_class=Image
     tracks [
      {
       type=MovieSceneFloatTrack
       display_name=Render Opacity
       sections [{type=MovieSceneFloatSection, channels [None {type=MovieSceneScriptingFloatChannel, keys [[4,0],[6,0.5],[6.0003,1],[15,0.25]]}]}]
      }
      {
       type=MovieSceneWidgetMaterialTrack
       sections [{
         type=MovieSceneParameterSection
         channels [
          FlareDensity {type=MovieSceneScriptingFloatChannel, keys [[4,1.5],[7,2],[15,2]]}
          FlareRadius {type=MovieSceneScriptingFloatChannel, keys [[4,0.75],[7,1.175]]}
         ]
        }]
      }
     ]
    }
    Reward_Preview {
     id=#7
     possessed_class=FortLazyImage
     tracks [
      {
       type=MovieSceneFloatTrack
       display_name=Render Opacity
       sections [{type=MovieSceneFloatSection, channels [None {type=MovieSceneScriptingFloatChannel, keys [[4,0],[6,1]]}]}]
      }
      {
       type=MovieScene2DTransformTrack
       sections [{
         type=MovieScene2DTransformSection
         channels [
          Translation.X {type=MovieSceneScriptingFloatChannel}
          Translation.Y {type=MovieSceneScriptingFloatChannel}
          Angle {type=MovieSceneScriptingFloatChannel}
          Scale.X {type=MovieSceneScriptingFloatChannel, keys [[4,1.5],[6,1],[7,1]]}
          Scale.Y {type=MovieSceneScriptingFloatChannel, keys [[4,1.5],[6,1],[7,1]]}
          Shear.X {type=MovieSceneScriptingFloatChannel}
          Shear.Y {type=MovieSceneScriptingFloatChannel}
         ]
        }]
      }
     ]
    }
    Text_RewardType {
     id=#8
     possessed_class=CommonTextBlock
     tracks [
      {
       type=MovieSceneFloatTrack
       display_name=Render Opacity
       sections [{type=MovieSceneFloatSection, channels [None {type=MovieSceneScriptingFloatChannel, keys [[0,0],[5,1]]}]}]
      }
      {
       type=MovieScene2DTransformTrack
       sections [{
         type=MovieScene2DTransformSection
         channels [
          Translation.X {type=MovieSceneScriptingFloatChannel, keys [[0,-32],[5,0]]}
          Translation.Y {type=MovieSceneScriptingFloatChannel, const [0,2]}
          Angle {type=MovieSceneScriptingFloatChannel}
          Scale.X {type=MovieSceneScriptingFloatChannel}
          Scale.Y {type=MovieSceneScriptingFloatChannel}
          Shear.X {type=MovieSceneScriptingFloatChannel}
          Shear.Y {type=MovieSceneScriptingFloatChannel}
         ]
        }]
      }
     ]
    }
   ]}
 }
 Outro {
  binding_count=1
  length_seconds=0.25
  object=&self:Outro
  sequence {bindings [Overlay_0 {
     id=#9
     possessed_class=Overlay
     tracks [{
       type=MovieSceneFloatTrack
       display_name=Render Opacity
       sections [{type=MovieSceneFloatSection, channels [None {type=MovieSceneScriptingFloatChannel, keys [[0,1],[5,0]]}]}]
      }]
    }]}
 }
]
graphs [
 EventGraph {nodes [
   n1 = OnIntroFinished() {is=title K2Node_CustomEvent}
   > FinishNotificationIntro(self:Fort Notification Entry Object Reference)
   n2 = OnOutroFinished() {is=title K2Node_CustomEvent}
   > FinishNotificationOutro(self:Fort Notification Entry Object Reference)
   Event Construct()
   n3 = GetOwningPlayer()
   > Is Valid(InputObject=@n3)
   > n4 = GetDynamicMaterial(self:Border Object Reference=Get CommonBorder_Content() -> (CommonBorder_Content:Common Border Object Reference))
   n5 = Get Class(Object=@n3) -> (ReturnValue:Player Controller Class Reference)
   > SetScalarParameterValue(
    self=@n4
    ParameterName=Container-Opacity
    Value=Select(Option 0:float=0.5, Option 1:float=0.85, Index=ClassIsChildOf(TestClass=@n5, ParentClass=/Script/FortniteGame.FortPlayerControllerFrontEnd)) -> (ReturnValue:float)
   )
   > BindToAnimationFinished(Animation=Get Intro(), Delegate=@n1)
   > BindToAnimationFinished(Animation=Get Outro() -> (Outro:Widget Animation Object Reference), Delegate=@n2)
   Event StartNotificationIntro() {is=title K2Node_Event}
   > PlayAnimationForward(InAnimation=Get Intro())
   > Do Once()
   > SpawnSound2D!(
    execute
    Sound:Sound Base Object Reference=/Game/Sounds/Fort_UI_Sounds/SharedQuests/JobBoard_Whoosh_Cue
    VolumeMultiplier:float=1
    PitchMultiplier:float=1
    StartTime:float=0
    ConcurrencySettings:Sound Concurrency Object Reference
    bPersistAcrossLevelTransition:bool=false
    bAutoDestroy:bool=true
   ) -> (then, ReturnValue:Audio Component Object Reference)
   n6 = Event UpdateQuestRewardVisuals!() -> (OutputDelegate, then, RewardItem:Item Definition Base Object Reference, QuestRarity:EFortRarity Enum=Common) {is=title K2Node_Event}
   > SetStyle_QuestCosmetic!(execute, self, Rarity:EFortRarity Enum=@n6.QuestRarity) -> (then)
   Event UpdateLevelUpVisuals!() -> (OutputDelegate, then, RewardItem:Fort Item Definition Object Reference) {is=title K2Node_Event}
   > SetStyle_LevelUp()
   Event StartNotificationOutro() {is=title K2Node_Event}
   > PlayAnimationForward(InAnimation=Get Outro())
   Event PreConstruct()
   > n7 = Branch(Condition=IsMobileGame())
   > SetUserSpecifiedScale!(
    execute
    self:Scale Box Object Reference=Get ScaleBox_Mobile() -> (ScaleBox_Mobile:Scale Box Object Reference)
    InUserSpecifiedScale:float=1.3
   ) -> (then)
   SetUserSpecifiedScale(execute=@n7.else, self=Get ScaleBox_Mobile(), InUserSpecifiedScale=1)
  ]}
 SetStyle_LevelUp {nodes [
   SetStyle_LevelUp() {is=title K2Node_FunctionEntry}
   > SetTextTransformPolicy!(
    execute
    self:Rich Text Block Object Reference=Get Text_Header() -> (Text_Header:Common Rich Text Block Object Reference)
    InTransformPolicy:ETextTransformPolicy Enum=ToUpper
   ) -> (then)
   > n1 = GetDynamicMaterial(self=Get CommonBorder_Content())
   > SetVectorParameterValue(self=@n1, ParameterName=RGB-Rarity1, Value=(R=0.2122,G=0.6514,B=1,A=1))
  ]}
 SetStyle_QuestCosmetic {nodes [
   SetStyle_QuestCosmetic!() -> (then, Rarity:EFortRarity Enum=Common) {is=title K2Node_FunctionEntry}
   > SetTextTransformPolicy(self=Get Text_Header(), InTransformPolicy=None)
   > n1 = GetDynamicMaterial(self=Get CommonBorder_Content())
   > SetVectorParameterValue(self=@n1, ParameterName=RGB-Rarity1, Value=(R=0.2122,G=0.6514,B=1,A=1))
  ]}
]
variables [bIsQuestCosmetic {type=boolean}]
widgets {
 info {inheritedWidgetCount=5, parentClass=/Script/FortniteUI.FortNotificationEntry_LevelUp, rootWidgetClass=&ScaleBox, widgetCount=14}
 widgets [
  ScaleBox_Mobile {widgetClassPath=&ScaleBox, bIsVariable=true, widget=ScaleBox_Mobile, properties {stretch=UserSpecified}}
  Overlay_0 {widgetClassPath=&Overlay, parent=ScaleBox_Mobile, widget=Overlay_0, slot {class=ScaleBoxSlot}}
  VerticalBox_1 {widgetClassPath=&VerticalBox, parent=Overlay_0, widget=VerticalBox_1, slot {class=OverlaySlot}}
  Text_RewardType {
   widgetClassPath=&CommonTextBlock
   bInherited=true
   bIsVariable=true
   parent=VerticalBox_1
   widget=Text_RewardType
   properties {
    text="LEVEL UP!"
    textTransformPolicy=ToUpper
    font {
     bForceMonospaced=false
     bMaterialIsStencil=false
     fontMaterial=null
     fontObject=&HeadingNow
     letterSpacing=0
     monospacedWidth=1
     size=12
     skewAmount=0
     typefaceFontName=86 Bold
     outlineSettings {
      bApplyOutlineToDropShadows=false
      bMiteredCorners=false
      bSeparateFillAlpha=false
      outlineMaterial=null
      outlineSize=0
      outlineColor=<rgba 0 0 0 1>
     }
    }
    renderTransform {angle=0, scale=<1 1>, shear=<-12 0>, translation=<0 0>}
    renderTransformPivot=<0 0.5>
    shadowOffset=<3 2>
   }
   slot {class=VerticalBoxSlot, properties {padding=<ltrb 0 0 135 10>}}
  }
  SizeBox_0 {
   widgetClassPath=&SizeBox
   parent=VerticalBox_1
   widget=SizeBox_0
   properties {bOverride_MinDesiredHeight=true, bOverride_WidthOverride=true, minDesiredHeight=120, minDesiredWidth=480, widthOverride=544}
   slot {class=VerticalBoxSlot}
  }
  CommonBorder_Content {
   widgetClassPath=&CommonBorder
   bIsVariable=true
   parent=SizeBox_0
   widget=CommonBorder_Content
   properties {
    background {
     bIsDynamicallyLoaded=false
     drawAs=Box
     imageType=NoImage
     mirroring=NoMirror
     resourceName=None
     resourceObject=/Game/Athena/HUD/Quests/Art/M_UI_Quest_Container_Toast
     tiling=NoTile
     imageSize=<64 64>
     margin=<ltrb 0.5 0.5 0.5 0.5>
     outlineSettings {
      bUseBrushTransparency=false
      roundingType=HalfHeightRadius
      width=0
      color {colorUseRule=UseColor_Specified, specifiedColor=<rgba 0 0 0 0>}
      cornerRadii=<0 0 0 0>
     }
     tintColor {colorUseRule=UseColor_Specified, specifiedColor=<rgba 1 1 1 1>}
     uVRegion {bIsValid=false, max=<0 0>, min=<0 0>}
    }
    padding=<ltrb 18 4 2 4>
    renderTransformPivot=<0 0.5>
   }
   slot {class=SizeBoxSlot, properties {padding=<ltrb 0 0 0 8>}}
  }
  Overlay_2 {
   widgetClassPath=&Overlay
   parent=CommonBorder_Content
   widget=Overlay_2
   slot {class=BorderSlot, properties {padding=<ltrb 18 4 2 4>}}
  }
  RewardFlare {
   widgetClassPath=&Image
   bIsVariable=true
   parent=Overlay_2
   widget=RewardFlare
   properties {
    brush {
     bIsDynamicallyLoaded=false
     drawAs=Image
     imageType=NoImage
     mirroring=NoMirror
     resourceName=None
     resourceObject=/Game/UI/Foundation/Onboarding/Materials/M_FX_TriangleScroll_Toast
     tiling=NoTile
     imageSize=<64 64>
     margin=<ltrb 0 0 0 0>
     outlineSettings {
      bUseBrushTransparency=false
      roundingType=HalfHeightRadius
      width=0
      color {colorUseRule=UseColor_Specified, specifiedColor=<rgba 0 0 0 0>}
      cornerRadii=<0 0 0 0>
     }
     tintColor {colorUseRule=UseColor_Specified, specifiedColor=<rgba 1 1 1 1>}
     uVRegion {bIsValid=false, max=<0 0>, min=<0 0>}
    }
    renderTransform {angle=0, scale=<-1 1>, shear=<0 0>, translation=<0 0>}
   }
   slot {class=OverlaySlot, properties {horizontalAlignment=HAlign_Fill, verticalAlignment=VAlign_Fill}}
  }
  VerticalBox_0 {
   widgetClassPath=&VerticalBox
   parent=Overlay_2
   widget=VerticalBox_0
   slot {class=OverlaySlot, properties {horizontalAlignment=HAlign_Fill, verticalAlignment=VAlign_Center, padding=<ltrb 16 8 120 8>}}
  }
  Text_Header {
   widgetClassPath=&CommonRichTextBlock
   bInherited=true
   bIsVariable=true
   parent=VerticalBox_0
   widget=Text_Header
   properties {
    autoWrapText=true
    bAutoCollapseWithEmptyText=true
    text="Header <levelNo>text</>"
    textStyleSet=&/CosmeticNotificationRichTextStyleSet
    decoratorClasses [/Game/Athena/HUD/QuestIcons/BP_QuestTile_IconDecorator.BP_QuestTile_IconDecorator_C]
    defaultTextStyleOverride {
     overflowPolicy=Clip
     transformPolicy=None
     colorAndOpacity {colorUseRule=UseColor_Specified, specifiedColor=<rgba 1 1 1 1>}
     font {
      bForceMonospaced=false
      bMaterialIsStencil=false
      fontMaterial=null
      fontObject=&NotoSans
      letterSpacing=0
      monospacedWidth=1
      size=24
      skewAmount=0
      typefaceFontName=NotoSans-Regular
      outlineSettings {
       bApplyOutlineToDropShadows=false
       bMiteredCorners=false
       bSeparateFillAlpha=false
       outlineMaterial=null
       outlineSize=0
       outlineColor=<rgba 0 0 0 1>
      }
     }
     highlightColor {colorUseRule=UseColor_Specified, specifiedColor=<rgba 0 0 0 1>}
     highlightShape {
      bIsDynamicallyLoaded=false
      drawAs=Image
      imageType=NoImage
      mirroring=NoMirror
      resourceName=None
      resourceObject=null
      tiling=NoTile
      imageSize=<32 32>
      margin=<ltrb 0 0 0 0>
      outlineSettings {
       bUseBrushTransparency=false
       roundingType=HalfHeightRadius
       width=0
       color {colorUseRule=UseColor_Specified, specifiedColor=<rgba 0 0 0 0>}
       cornerRadii=<0 0 0 0>
      }
      tintColor {colorUseRule=UseColor_Specified, specifiedColor=<rgba 1 1 1 1>}
      uVRegion {bIsValid=false, max=<0 0>, min=<0 0>}
     }
     selectedBackgroundColor {colorUseRule=UseColor_Foreground, specifiedColor=<rgba 1 0 1 1>}
     shadowColorAndOpacity=<rgba 0 0 0 1>
     shadowOffset=<0 0>
     strikeBrush {
      bIsDynamicallyLoaded=false
      drawAs=Image
      imageType=NoImage
      mirroring=NoMirror
      resourceName=None
      resourceObject=null
      tiling=NoTile
      imageSize=<32 32>
      margin=<ltrb 0 0 0 0>
      outlineSettings {
       bUseBrushTransparency=false
       roundingType=HalfHeightRadius
       width=0
       color {colorUseRule=UseColor_Specified, specifiedColor=<rgba 0 0 0 0>}
       cornerRadii=<0 0 0 0>
      }
      tintColor {colorUseRule=UseColor_Specified, specifiedColor=<rgba 1 1 1 1>}
      uVRegion {bIsValid=false, max=<0 0>, min=<0 0>}
     }
     underlineBrush {
      bIsDynamicallyLoaded=false
      drawAs=Image
      imageType=NoImage
      mirroring=NoMirror
      resourceName=None
      resourceObject=null
      tiling=NoTile
      imageSize=<32 32>
      margin=<ltrb 0 0 0 0>
      outlineSettings {
       bUseBrushTransparency=false
       roundingType=HalfHeightRadius
       width=0
       color {colorUseRule=UseColor_Specified, specifiedColor=<rgba 0 0 0 0>}
       cornerRadii=<0 0 0 0>
      }
      tintColor {colorUseRule=UseColor_Specified, specifiedColor=<rgba 1 1 1 1>}
      uVRegion {bIsValid=false, max=<0 0>, min=<0 0>}
     }
    }
   }
   slot {class=VerticalBoxSlot, properties {verticalAlignment=VAlign_Center, size {sizeRule=Automatic, value=0.5}}}
  }
  Text_Description {
   widgetClassPath=&CommonTextBlock
   bInherited=true
   bIsVariable=true
   parent=VerticalBox_0
   widget=Text_Description
   properties {
    autoWrapText=true
    text=Gained Reward Description
    textTransformPolicy=ToUpper
    font {
     bForceMonospaced=false
     bMaterialIsStencil=false
     fontMaterial=null
     fontObject=&HeadingNow
     letterSpacing=0
     monospacedWidth=1
     size=15
     skewAmount=0
     typefaceFontName=86 Bold Italic
     outlineSettings {
      bApplyOutlineToDropShadows=false
      bMiteredCorners=false
      bSeparateFillAlpha=false
      outlineMaterial=null
      outlineSize=0
      outlineColor=<rgba 0 0 0 1>
     }
    }
    shadowOffset=<0 0>
   }
   slot {class=VerticalBoxSlot}
  }
  Reward_Glow {
   widgetClassPath=&Image
   bIsVariable=true
   parent=Overlay_0
   widget=Reward_Glow
   properties {
    renderOpacity=0.1
    brush {
     bIsDynamicallyLoaded=false
     drawAs=Image
     imageType=NoImage
     mirroring=NoMirror
     resourceName=None
     resourceObject=/Game/UI/Foundation/Onboarding/Materials/M_FX_Flare_Master_Toast
     tiling=NoTile
     imageSize=<128 128>
     margin=<ltrb 0 0 0 0>
     outlineSettings {
      bUseBrushTransparency=false
      roundingType=HalfHeightRadius
      width=0
      color {colorUseRule=UseColor_Specified, specifiedColor=<rgba 0 0 0 0>}
      cornerRadii=<0 0 0 0>
     }
     tintColor {colorUseRule=UseColor_Specified, specifiedColor=<rgba 1 1 1 1>}
     uVRegion {bIsValid=false, max=<0 0>, min=<0 0>}
    }
    colorAndOpacity=<rgba 1 1 1 0.4>
    renderTransform {angle=0, scale=<2 1.5>, shear=<0 0>, translation=<0 0>}
   }
   slot {class=OverlaySlot, properties {horizontalAlignment=HAlign_Right, verticalAlignment=VAlign_Bottom}}
  }
  Reward_Preview_Shadow {
   widgetClassPath=&FortLazyImage
   bInherited=true
   bIsVariable=true
   parent=Overlay_0
   widget=Reward_Preview_Shadow
   properties {
    designerPreviewItem=/Game/Athena/Items/Cosmetics/Pickaxes/DefaultPickaxe
    brush {
     bIsDynamicallyLoaded=false
     drawAs=Image
     imageType=NoImage
     mirroring=NoMirror
     resourceName=None
     resourceObject=/Game/UI/Foundation/Textures/Icons/Weapons/Items/T-Icon-Pickaxes-DefaultMarkIIIPickaxe
     tiling=NoTile
     imageSize=<80 80>
     margin=<ltrb 0 0 0 0>
     outlineSettings {
      bUseBrushTransparency=false
      roundingType=HalfHeightRadius
      width=0
      color {colorUseRule=UseColor_Specified, specifiedColor=<rgba 0 0 0 0>}
      cornerRadii=<0 0 0 0>
     }
     tintColor {colorUseRule=UseColor_Specified, specifiedColor=<rgba 0 0 0 1>}
     uVRegion {bIsValid=false, max=<0 0>, min=<0 0>}
    }
    colorAndOpacity=<rgba 1 1 1 0.35>
   }
   slot {class=OverlaySlot, properties {horizontalAlignment=HAlign_Right, verticalAlignment=VAlign_Center, padding=<ltrb 0 10 24 0>}}
  }
  Reward_Preview {
   widgetClassPath=&FortLazyImage
   bInherited=true
   bIsVariable=true
   parent=Overlay_0
   widget=Reward_Preview
   properties {
    designerPreviewItem=&DefaultPickaxe
    brush {
     bIsDynamicallyLoaded=false
     drawAs=Image
     imageType=NoImage
     mirroring=NoMirror
     resourceName=None
     resourceObject=&T-Icon-Pickaxes-DefaultMarkIIIPickaxe
     tiling=NoTile
     imageSize=<96 96>
     margin=<ltrb 0 0 0 0>
     outlineSettings {
      bUseBrushTransparency=false
      roundingType=HalfHeightRadius
      width=0
      color {colorUseRule=UseColor_Specified, specifiedColor=<rgba 0 0 0 0>}
      cornerRadii=<0 0 0 0>
     }
     tintColor {colorUseRule=UseColor_Specified, specifiedColor=<rgba 1 1 1 1>}
     uVRegion {bIsValid=false, max=<0 0>, min=<0 0>}
    }
   }
   slot {class=OverlaySlot, properties {horizontalAlignment=HAlign_Right, verticalAlignment=VAlign_Center, padding=<ltrb 0 12 20 0>}}
  }
 ]
}
