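# DefaultNotification is a FortNotificationEntry widget that serves as a default in-game event notification in the UI.
# A vertical box stacks a header text block above an indented description text block, both holding placeholder text.
# The intro and outro events play no animation; each only waits on a fixed Delay, then calls FinishNotificationIntro or FinishNotificationOutro.
# The graph never sets the text, and PreConstruct, Construct and Tick are empty.
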WidgetBlueprint /Game/UI/InGame/EventNotifications/DefaultNotification
graphs [EventGraph {nodes [
   Event StartNotificationIntro() {is=title K2Node_Event}
   > Delay(Duration=2)
   > FinishNotificationIntro(self:Fort Notification Entry Object Reference)
   Event StartNotificationOutro() {is=title K2Node_Event}
   > Delay(Duration=1)
   > FinishNotificationOutro(self:Fort Notification Entry Object Reference)
   Event PreConstruct()
   Event Construct()
   Event Tick()
  ]}]
variables []
widgets {
 info {inheritedWidgetCount=0, parentClass=/Script/FortniteUI.FortNotificationEntry, rootWidgetClass=&VerticalBox, widgetCount=3}
 widgets [
  VerticalBox_0 {widgetClassPath=&VerticalBox, widget=VerticalBox_0}
  Text_Header {
   widgetClassPath=&CommonTextBlock
   parent=VerticalBox_0
   widget=Text_Header
   properties {
    style=&TextStyle-Base_C
    text=Header
    font {
     bForceMonospaced=false
     bMaterialIsStencil=false
     fontMaterial=null
     fontObject=&NotoSans
     letterSpacing=0
     monospacedWidth=1
     size=24
     skewAmount=0
     typefaceFontName=NotoSans-Regular
     outlineSettings {
      bApplyOutlineToDropShadows=false
      bMiteredCorners=false
      bSeparateFillAlpha=false
      outlineMaterial=null
      outlineSize=0
      outlineColor=<rgba 0 0 0 1>
     }
    }
    shadowOffset=<0 0>
   }
   slot {class=VerticalBoxSlot}
  }
  Text_Description {
   widgetClassPath=&CommonTextBlock
   parent=VerticalBox_0
   widget=Text_Description
   properties {
    style=&TextStyle-Base_C
    text=Body Description
    font {
     bForceMonospaced=false
     bMaterialIsStencil=false
     fontMaterial=null
     fontObject=&NotoSans
     letterSpacing=0
     monospacedWidth=1
     size=24
     skewAmount=0
     typefaceFontName=NotoSans-Regular
     outlineSettings {
      bApplyOutlineToDropShadows=false
      bMiteredCorners=false
      bSeparateFillAlpha=false
      outlineMaterial=null
      outlineSize=0
      outlineColor=<rgba 0 0 0 1>
     }
    }
    shadowOffset=<0 0>
   }
   slot {class=VerticalBoxSlot, properties {horizontalAlignment=HAlign_Left, padding=<ltrb 50 0 0 0>}}
  }
 ]
}
