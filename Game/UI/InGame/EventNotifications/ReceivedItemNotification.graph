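# Toast widget for the received-item notification, showing item name, description and preview icon over an animated, rarity-colored backing, glow and flare.
# UpdateVisuals feeds the item's rarity to RarityVisuals, which sets color parameters on those three dynamic materials; design-time PreConstruct previews this using DevRarity.
# Intro plays Anim_RewardUnlocked and signals the parent on finish; outro plays Outro, then signals after a matching Delay.
# The Intro animation and IntroSpeed variable are unused in the graph.
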WidgetBlueprint /Game/UI/InGame/EventNotifications/ReceivedItemNotification
animations [
 Anim_RewardUnlocked {
  binding_count=7
  length_seconds=1.75
  object=&self:Anim_RewardUnlocked
  sequence {bindings [
    Overlay_5 {
     id=#1
     possessed_class=Overlay
     tracks [{
       type=MovieSceneFloatTrack
       display_name=Render Opacity
       sections [{type=MovieSceneFloatSection, start_frame=0, channels [None {type=MovieSceneScriptingFloatChannel, keys [[0,0],[1,1]]}]}]
      }]
    }
    Backing {
     id=#2
     possessed_class=CommonBorder
     tracks [
      {
       type=MovieSceneFloatTrack
       display_name=Render Opacity
       sections [{type=MovieSceneFloatSection, channels [None {type=MovieSceneScriptingFloatChannel, keys [[1,0],[6,1],[35,1]]}]}]
      }
      {
       type=MovieScene2DTransformTrack
       sections [{
         type=MovieScene2DTransformSection
         channels [
          Translation.X {type=MovieSceneScriptingFloatChannel, keys [[1,-64],[6,0]]}
          Translation.Y {type=MovieSceneScriptingFloatChannel}
          Angle {type=MovieSceneScriptingFloatChannel}
          Scale.X {type=MovieSceneScriptingFloatChannel}
          Scale.Y {type=MovieSceneScriptingFloatChannel}
          Shear.X {type=MovieSceneScriptingFloatChannel}
          Shear.Y {type=MovieSceneScriptingFloatChannel}
         ]
        }]
      }
     ]
    }
    Text_Description {
     id=#3
     possessed_class=CommonTextBlock
     tracks [{
       type=MovieScene2DTransformTrack
       sections [{
         type=MovieScene2DTransformSection
         channels [
          Translation.X {type=MovieSceneScriptingFloatChannel, keys [[1,-32],[6,0]]}
          Translation.Y {type=MovieSceneScriptingFloatChannel, const [0,1]}
          Angle {type=MovieSceneScriptingFloatChannel}
          Scale.X {type=MovieSceneScriptingFloatChannel}
          Scale.Y {type=MovieSceneScriptingFloatChannel}
          Shear.X {type=MovieSceneScriptingFloatChannel}
          Shear.Y {type=MovieSceneScriptingFloatChannel}
         ]
        }]
      }]
    }
    Reward_Glow {
     id=#4
     possessed_class=Image
     tracks [
      {
       type=MovieSceneFloatTrack
       display_name=Render Opacity
       sections [{type=MovieSceneFloatSection, channels [None {type=MovieSceneScriptingFloatChannel, keys [[4,0],[6,1],[35,0.25]]}]}]
      }
      {
       type=MovieScene2DTransformTrack
       sections [{
         type=MovieScene2DTransformSection
         channels [
          Translation.X {type=MovieSceneScriptingFloatChannel}
          Translation.Y {type=MovieSceneScriptingFloatChannel}
          Angle {type=MovieSceneScriptingFloatChannel}
          Scale.X {type=MovieSceneScriptingFloatChannel, keys [[4,1.5],[6,2],[7,2]]}
          Scale.Y {type=MovieSceneScriptingFloatChannel, keys [[4,0.75],[6,1.5],[7,1.5]]}
          Shear.X {type=MovieSceneScriptingFloatChannel}
          Shear.Y {type=MovieSceneScriptingFloatChannel}
         ]
        }]
      }
     ]
    }
    Reward_Preview {
     id=#5
     possessed_class=FortLazyImage
     tracks [
      {
       type=MovieSceneFloatTrack
       display_name=Render Opacity
       sections [{type=MovieSceneFloatSection, channels [None {type=MovieSceneScriptingFloatChannel, keys [[4,0],[6,1]]}]}]
      }
      {
       type=MovieScene2DTransformTrack
       sections [{
         type=MovieScene2DTransformSection
         channels [
          Translation.X {type=MovieSceneScriptingFloatChannel}
          Translation.Y {type=MovieSceneScriptingFloatChannel}
          Angle {type=MovieSceneScriptingFloatChannel}
          Scale.X {type=MovieSceneScriptingFloatChannel, keys [[4,1.5],[6,1],[7,1]]}
          Scale.Y {type=MovieSceneScriptingFloatChannel, keys [[4,1.5],[6,1],[7,1]]}
          Shear.X {type=MovieSceneScriptingFloatChannel}
          Shear.Y {type=MovieSceneScriptingFloatChannel}
         ]
        }]
      }
     ]
    }
    RewardFlare {
     id=#6
     possessed_class=Image
     tracks [
      {
       type=MovieSceneFloatTrack
       display_name=Render Opacity
       sections [{type=MovieSceneFloatSection, channels [None {type=MovieSceneScriptingFloatChannel, keys [[4,0],[6,0.5],[6.0003,1],[35,0.25]]}]}]
      }
      {
       type=MovieSceneWidgetMaterialTrack
       sections [{
         type=MovieSceneParameterSection
         channels [
          FlareDensity {type=MovieSceneScriptingFloatChannel, keys [[4,1.5],[7,2],[35,2]]}
          FlareRadius {type=MovieSceneScriptingFloatChannel, keys [[4,0.75],[7,1.175]]}
         ]
        }]
      }
     ]
    }
    Text_RewardType {
     id=#7
     possessed_class=CommonTextBlock
     tracks [
      {
       type=MovieSceneFloatTrack
       display_name=Render Opacity
       sections [{type=MovieSceneFloatSection, channels [None {type=MovieSceneScriptingFloatChannel, keys [[0,0],[5,1]]}]}]
      }
      {
       type=MovieScene2DTransformTrack
       sections [{
         type=MovieScene2DTransformSection
         channels [
          Translation.X {type=MovieSceneScriptingFloatChannel, keys [[0,-32],[5,0]]}
          Translation.Y {type=MovieSceneScriptingFloatChannel, const [0,2]}
          Angle {type=MovieSceneScriptingFloatChannel}
          Scale.X {type=MovieSceneScriptingFloatChannel}
          Scale.Y {type=MovieSceneScriptingFloatChannel}
          Shear.X {type=MovieSceneScriptingFloatChannel}
          Shear.Y {type=MovieSceneScriptingFloatChannel}
         ]
        }]
      }
     ]
    }
   ]}
 }
 Intro {
  binding_count=1
  length_seconds=0.5
  object=&self:Intro
  sequence {bindings [SizeBox_1 {
     id=#8
     display_name=Overlay_5
     possessed_class=SizeBox
     tracks [
      {
       type=MovieSceneFloatTrack
       display_name=Render Opacity
       sections [{type=MovieSceneFloatSection, channels [None {type=MovieSceneScriptingFloatChannel, keys [[0,0],[5,1]]}]}]
      }
      {
       type=MovieScene2DTransformTrack
       sections [{
         type=MovieScene2DTransformSection
         channels [
          Translation.X {type=MovieSceneScriptingFloatChannel, keys [[0,-50],[10,0]]}
          Translation.Y {type=MovieSceneScriptingFloatChannel}
          Angle {type=MovieSceneScriptingFloatChannel}
          Scale.X {type=MovieSceneScriptingFloatChannel}
          Scale.Y {type=MovieSceneScriptingFloatChannel}
          Shear.X {type=MovieSceneScriptingFloatChannel}
          Shear.Y {type=MovieSceneScriptingFloatChannel}
         ]
        }]
      }
     ]
    }]}
 }
 Outro {
  binding_count=1
  length_seconds=0.25
  object=&self:Outro
  sequence {bindings [SizeBox_1 {
     id=#9
     display_name=Overlay_5
     possessed_class=SizeBox
     tracks [{
       type=MovieSceneFloatTrack
       display_name=Render Opacity
       sections [{type=MovieSceneFloatSection, channels [None {type=MovieSceneScriptingFloatChannel, keys [[0,1],[5,0]]}]}]
      }]
    }]}
 }
]
graphs [EventGraph {nodes [
   n1 = Intro FInished() {is=title K2Node_CustomEvent}
   > FinishNotificationIntro(self:Fort Notification Entry Object Reference)
   n2 = RarityVisuals!() -> (OutputDelegate, then, Rarity:EFortRarity Enum=Common) {is=title K2Node_CustomEvent}
   > n3 = GetRarityData!(execute, Rarity:EFortRarity Enum=@n2.Rarity) -> (then, ReturnValue:Fort Rarity Item Data Structure)
   > n4 = GetDynamicMaterial(self:Border Object Reference=Get Backing() -> (Backing:Common Border Object Reference))
   n5 = Break Fort Rarity Item Data!(FortRarityItemData:"Fort Rarity Item Data Structure (by ref)"=@n3) -> (
    Name:Text
    Color1:Linear Color Structure
    Color2:Linear Color Structure
    Color3:Linear Color Structure
    Color4:Linear Color Structure
    Color5:Linear Color Structure
    Radius:float=~
    Falloff:float=~
    Brightness:float=~
    Roughness:float=~
    Glow:float=~
   ) {is=title K2Node_BreakStruct}
   > SetVectorParameterValue(self=@n4, ParameterName=RGB-Rarity1, Value=@n5.Color2)
   > SetVectorParameterValue(self=@n4, ParameterName=Background_color, Value=@n5.Color4)
   > SetVectorParameterValue(self=@n4, ParameterName=Triangle_color, Value=@n5.Color3)
   > n6 = GetDynamicMaterial(self:Image Object Reference=Get Reward_Glow() -> (Reward_Glow:Image Object Reference))
   n7 = "Lerp (LinearColor)"(A=@n5.Color1, B=(R=1,G=1,B=1,A=1), Alpha=0.6)
   > SetVectorParameterValue(self=@n6, ParameterName=RGB 2, Value=@n7)
   > SetVectorParameterValue(self=@n6, ParameterName=RGB 1, Value=@n5.Color3)
   > n8 = GetDynamicMaterial(self=Get RewardFlare() -> (RewardFlare:Image Object Reference))
   > SetVectorParameterValue(self=@n8, ParameterName=Triangle_RGB_B, Value=@n7)
   > SetVectorParameterValue(self=@n8, ParameterName=RGB_InnerFlame, Value=@n5.Color1)
   > SetVectorParameterValue(self=@n8, ParameterName=Triangle_RGB_A, Value=@n5.Color1)
   > SetVectorParameterValue(self=@n8, ParameterName=Triangle_RGB_A, Value=@n5.Color2)
   n9 = Event PreConstruct()
   > Branch(Condition=@n9.IsDesignTime)
   > RarityVisuals!(execute, self, Rarity:EFortRarity Enum=Get DevRarity() -> (DevRarity:EFortRarity Enum=Common)) -> (then) {is=title K2Node_CallFunction}
   Event Construct()
   > BindToAnimationFinished(Animation=Get Anim_RewardUnlocked() -> (Anim_RewardUnlocked:Widget Animation Object Reference), Delegate=@n1)
   Event StartNotificationIntro() {is=title K2Node_Event}
   > PlayAnimation(InAnimation=Get Anim_RewardUnlocked())
   Event StartNotificationOutro() {is=title K2Node_Event}
   n10 = Get Outro() -> (Outro:Widget Animation Object Reference)
   > PlayAnimation(InAnimation=@n10)
   > Delay(Duration=GetEndTime(self=@n10))
   > FinishNotificationOutro(self:Fort Notification Entry Object Reference)
   n11 = Event UpdateVisuals!() -> (OutputDelegate, then, RewardItem:Item Definition Base Object Reference, ItemQuantity:int=0) {is=title K2Node_Event}
   > RarityVisuals(Rarity=GetRarity!(ItemDefinition:Item Definition Base Object Reference=@n11.RewardItem) -> (ReturnValue:EFortRarity Enum=Common))
  ]}]
variables [
 DevRarity {type=EFortRarity, default=Epic}
 IntroSpeed {type=number, default=3}
]
widgets {
 info {
  inheritedWidgetCount=5
  parentClass=/Script/FortniteUI.FortNotificationEntry_ReceivedItem
  rootWidgetClass=&Overlay
  widgetCount=18
 }
 widgets [
  Overlay_5 {widgetClassPath=&Overlay, widget=Overlay_5}
  VerticalBox {widgetClassPath=&VerticalBox, parent=Overlay_5, widget=VerticalBox, slot {class=OverlaySlot}}
  Text_RewardType {
   widgetClassPath=&CommonTextBlock
   bInherited=true
   bIsVariable=true
   parent=VerticalBox
   widget=Text_RewardType
   properties {
    text=Received
    textTransformPolicy=ToUpper
    visibility=Collapsed
    font {
     bForceMonospaced=false
     bMaterialIsStencil=false
     fontMaterial=null
     fontObject=&Burbank
     letterSpacing=0
     monospacedWidth=1
     size=22
     skewAmount=0
     typefaceFontName=Burbank-PlayerPower
     outlineSettings {
      bApplyOutlineToDropShadows=false
      bMiteredCorners=false
      bSeparateFillAlpha=true
      outlineMaterial=&M-FontShadow
      outlineSize=0
      outlineColor=<rgba 0 0 0 1>
     }
    }
    renderTransform {angle=0, scale=<1 1>, shear=<-12 0>, translation=<0 0>}
    shadowColorAndOpacity=<rgba 0 0 0 0.5>
    shadowOffset=<3 2>
   }
   slot {class=VerticalBoxSlot, properties {padding=<ltrb 18 0 0 0>}}
  }
  SizeBox_4 {
   widgetClassPath=&SizeBox
   parent=VerticalBox
   widget=SizeBox_4
   properties {bOverride_MinDesiredHeight=true, bOverride_WidthOverride=true, minDesiredHeight=104, widthOverride=448}
   slot {class=VerticalBoxSlot}
  }
  Backing {
   widgetClassPath=&CommonBorder
   bIsVariable=true
   parent=SizeBox_4
   widget=Backing
   properties {
    background {
     bIsDynamicallyLoaded=false
     drawAs=Box
     imageType=NoImage
     mirroring=NoMirror
     resourceName=None
     resourceObject=/Game/Athena/HUD/Quests/Art/M_UI_NPC_Freebie_Container
     tiling=NoTile
     imageSize=<48 48>
     margin=<ltrb 0.5 0.5 0.5 0.5>
     outlineSettings {
      bUseBrushTransparency=false
      roundingType=HalfHeightRadius
      width=0
      color {colorUseRule=UseColor_Specified, specifiedColor=<rgba 0 0 0 0>}
      cornerRadii=<0 0 0 0>
     }
     tintColor {colorUseRule=UseColor_Specified, specifiedColor=<rgba 1 1 1 1>}
     uVRegion {bIsValid=false, max=<0 0>, min=<0 0>}
    }
    padding=<ltrb 0 0 0 0>
   }
   slot {class=SizeBoxSlot}
  }
  Overlay_3 {widgetClassPath=&Overlay, parent=Backing, widget=Overlay_3, slot {class=BorderSlot, properties {padding=<ltrb 0 0 0 0>}}}
  RewardFlare {
   widgetClassPath=&Image
   bIsVariable=true
   parent=Overlay_3
   widget=RewardFlare
   properties {
    brush {
     bIsDynamicallyLoaded=false
     drawAs=Image
     imageType=NoImage
     mirroring=NoMirror
     resourceName=None
     resourceObject=/Game/UI/Foundation/Onboarding/Materials/M_FX_TriangleScroll_Toast
     tiling=NoTile
     imageSize=<64 64>
     margin=<ltrb 0 0 0 0>
     outlineSettings {
      bUseBrushTransparency=false
      roundingType=HalfHeightRadius
      width=0
      color {colorUseRule=UseColor_Specified, specifiedColor=<rgba 0 0 0 0>}
      cornerRadii=<0 0 0 0>
     }
     tintColor {colorUseRule=UseColor_Specified, specifiedColor=<rgba 1 1 1 1>}
     uVRegion {bIsValid=false, max=<0 0>, min=<0 0>}
    }
    renderTransform {angle=0, scale=<-1 1>, shear=<0 0>, translation=<0 0>}
   }
   slot {class=OverlaySlot, properties {horizontalAlignment=HAlign_Fill, verticalAlignment=VAlign_Fill, padding=<ltrb 0 0 2 0>}}
  }
  VerticalBox_93 {
   widgetClassPath=&VerticalBox
   parent=Overlay_3
   widget=VerticalBox_93
   slot {class=OverlaySlot, properties {verticalAlignment=VAlign_Center, padding=<ltrb 24 0 0 0>}}
  }
  SizeBox_1 {
   widgetClassPath=&SizeBox
   parent=VerticalBox_93
   widget=SizeBox_1
   properties {bOverride_WidthOverride=true, widthOverride=296}
   slot {class=VerticalBoxSlot, properties {horizontalAlignment=HAlign_Left, verticalAlignment=VAlign_Center}}
  }
  ScaleBox_1 {
   widgetClassPath=&ScaleBox
   parent=SizeBox_1
   widget=ScaleBox_1
   properties {stretch=ScaleToFitX, stretchDirection=DownOnly}
   slot {class=SizeBoxSlot}
  }
  Text_Header {
   widgetClassPath=&CommonTextBlock
   bInherited=true
   bIsVariable=true
   parent=ScaleBox_1
   widget=Text_Header
   properties {
    scrollStyle=&TextScrollStyle-Base_C
    text=Item Name
    textTransformPolicy=ToUpper
    font {
     bForceMonospaced=false
     bMaterialIsStencil=false
     fontMaterial=null
     fontObject=&HeadingNow
     letterSpacing=0
     monospacedWidth=1
     size=15
     skewAmount=0
     typefaceFontName=86 Bold Italic
     outlineSettings {
      bApplyOutlineToDropShadows=false
      bMiteredCorners=false
      bSeparateFillAlpha=true
      outlineMaterial=null
      outlineSize=0
      outlineColor=<rgba 0 0 0 1>
     }
    }
    shadowOffset=<0 0>
   }
   slot {class=ScaleBoxSlot, properties {horizontalAlignment=HAlign_Left}}
  }
  SizeBox_0 {
   widgetClassPath=&SizeBox
   parent=VerticalBox_93
   widget=SizeBox_0
   properties {bOverride_WidthOverride=true, widthOverride=296}
   slot {class=VerticalBoxSlot, properties {verticalAlignment=VAlign_Center}}
  }
  ScaleBox_0 {
   widgetClassPath=&ScaleBox
   parent=SizeBox_0
   widget=ScaleBox_0
   properties {stretch=ScaleToFitX, stretchDirection=DownOnly}
   slot {class=SizeBoxSlot}
  }
  Text_Description {
   widgetClassPath=&CommonTextBlock
   bInherited=true
   bIsVariable=true
   parent=ScaleBox_0
   widget=Text_Description
   properties {
    scrollStyle=&TextScrollStyle-Base_C
    text="RARITY {Y} / x{Count}"
    colorAndOpacity {colorUseRule=UseColor_Specified, specifiedColor=<rgba 1 1 1 0.6>}
    font {
     bForceMonospaced=false
     bMaterialIsStencil=false
     fontMaterial=null
     fontObject=&HeadingNow
     letterSpacing=0
     monospacedWidth=1
     size=11.25
     skewAmount=0
     typefaceFontName=75 Medium
     outlineSettings {
      bApplyOutlineToDropShadows=false
      bMiteredCorners=false
      bSeparateFillAlpha=true
      outlineMaterial=null
      outlineSize=0
      outlineColor=<rgba 0 0 0 0>
     }
    }
    shadowOffset=<0 0>
   }
   slot {class=ScaleBoxSlot, properties {horizontalAlignment=HAlign_Left}}
  }
  Overlay_1 {
   widgetClassPath=&Overlay
   parent=Overlay_5
   widget=Overlay_1
   slot {class=OverlaySlot, properties {horizontalAlignment=HAlign_Right, verticalAlignment=VAlign_Bottom}}
  }
  Reward_Glow {
   widgetClassPath=&Image
   bIsVariable=true
   parent=Overlay_1
   widget=Reward_Glow
   properties {
    renderOpacity=0.5
    brush {
     bIsDynamicallyLoaded=false
     drawAs=Image
     imageType=NoImage
     mirroring=NoMirror
     resourceName=None
     resourceObject=/Game/UI/Foundation/Onboarding/Materials/M_FX_Flare_Master_Toast
     tiling=NoTile
     imageSize=<96 96>
     margin=<ltrb 0 0 0 0>
     outlineSettings {
      bUseBrushTransparency=false
      roundingType=HalfHeightRadius
      width=0
      color {colorUseRule=UseColor_Specified, specifiedColor=<rgba 0 0 0 0>}
      cornerRadii=<0 0 0 0>
     }
     tintColor {colorUseRule=UseColor_Specified, specifiedColor=<rgba 1 1 1 1>}
     uVRegion {bIsValid=false, max=<0 0>, min=<0 0>}
    }
    renderTransform {angle=0, scale=<2 1.5>, shear=<0 0>, translation=<0 0>}
   }
   slot {class=OverlaySlot, properties {horizontalAlignment=HAlign_Right, verticalAlignment=VAlign_Bottom}}
  }
  Reward_Preview_Shadow {
   widgetClassPath=&FortLazyImage
   bInherited=true
   bIsVariable=true
   parent=Overlay_1
   widget=Reward_Preview_Shadow
   properties {
    designerPreviewItem=/Game/Athena/Items/Cosmetics/Pickaxes/DefaultPickaxe
    renderOpacity=0
    brush {
     bIsDynamicallyLoaded=false
     drawAs=Image
     imageType=NoImage
     mirroring=NoMirror
     resourceName=None
     resourceObject=/Game/UI/Foundation/Textures/Icons/Weapons/Items/T-Icon-Pickaxes-StrawberryPilot1HPickaxe
     tiling=NoTile
     imageSize=<104 104>
     margin=<ltrb 0 0 0 0>
     outlineSettings {
      bUseBrushTransparency=false
      roundingType=HalfHeightRadius
      width=0
      color {colorUseRule=UseColor_Specified, specifiedColor=<rgba 0 0 0 0>}
      cornerRadii=<0 0 0 0>
     }
     tintColor {colorUseRule=UseColor_Specified, specifiedColor=<rgba 0 0 0 1>}
     uVRegion {bIsValid=false, max=<0 0>, min=<0 0>}
    }
    colorAndOpacity=<rgba 1 1 1 0.35>
   }
   slot {class=OverlaySlot, properties {horizontalAlignment=HAlign_Right, verticalAlignment=VAlign_Bottom}}
  }
  Reward_Preview {
   widgetClassPath=&FortLazyImage
   bInherited=true
   bIsVariable=true
   parent=Overlay_1
   widget=Reward_Preview
   properties {
    designerPreviewItem=/Game/Athena/Items/Cosmetics/Characters/CID_001_Athena_Commando_F_Default
    brush {
     bIsDynamicallyLoaded=false
     drawAs=Image
     imageType=NoImage
     mirroring=NoMirror
     resourceName=None
     resourceObject=/Game/UI/Foundation/Textures/Icons/Heroes/Athena/Soldier/T-Soldier-HID-396-Athena-Commando-F-Swashbuckler
     tiling=NoTile
     imageSize=<104 104>
     margin=<ltrb 0 0 0 0>
     outlineSettings {
      bUseBrushTransparency=false
      roundingType=HalfHeightRadius
      width=0
      color {colorUseRule=UseColor_Specified, specifiedColor=<rgba 0 0 0 0>}
      cornerRadii=<0 0 0 0>
     }
     tintColor {colorUseRule=UseColor_Specified, specifiedColor=<rgba 1 1 1 1>}
     uVRegion {bIsValid=false, max=<0 0>, min=<0 0>}
    }
    loadingBackgroundBrush {
     bIsDynamicallyLoaded=false
     drawAs=Image
     imageType=NoImage
     mirroring=NoMirror
     resourceName=None
     resourceObject=/FortUILibrary/UIKit/Throbber/MI_UIKit_Monochrome_Throbber
     tiling=NoTile
     imageSize=<32 32>
     margin=<ltrb 0 0 0 0>
     outlineSettings {
      bUseBrushTransparency=false
      roundingType=HalfHeightRadius
      width=0
      color {colorUseRule=UseColor_Specified, specifiedColor=<rgba 0 0 0 0>}
      cornerRadii=<0 0 0 0>
     }
     tintColor {colorUseRule=UseColor_Specified, specifiedColor=<rgba 1 1 1 1>}
     uVRegion {bIsValid=false, max=<0 0>, min=<0 0>}
    }
   }
   slot {class=OverlaySlot, properties {horizontalAlignment=HAlign_Right, verticalAlignment=VAlign_Bottom, padding=<ltrb 0 0 8 0>}}
  }
 ]
}
